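; Fortnite NPC pawn for the medium TidalCrane swarm creature, handling threat sensing, health bar and cosmetic feedback.
; On authority, a nearby-actor perception listener sets a blackboard threat flag, and deep water triggers a drown ability.
; Remote instances answer death and attack gameplay events, writing timestamps to mesh custom primitive data for fade and attack materials.
; It also plays spawn, impact and pod weakpoint effects, refreshes custom depth on a timer, and sets encounter music.
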(event Custom|ForceUpdateCustomDepth
  (bind _fortpawncomponent_customdepth (Variables|Default|GetFortPawnComponent_CustomDepth))
  (if (CustomDepth|IsRenderCustomDepthEnabled _fortpawncomponent_customdepth)
    (CustomDepth|Update _fortpawncomponent_customdepth 0.0 true)))

(event EventBeginPlay
  (bind _returnvalue (CallFunction|GetLocationwithCapsuleHalfHeight))
  (|CallParentFunction)
  (CallFunction|SetupStatusWidget)
  (Utilities|FlowControl|Sequence
    (:then_0
      (Networking|SwitchHasAuthority
        (:Authority
          (CallFunction|SetupProximityComponent)
          (Utilities|FlowControl|Delay 2.0)
          (Variables|Default|SetInitialSpawnVFXDone true))
        (:Remote
          (Ability|Async|WaitGameplayEventToActor 0 "(TagName=\"Event.TidalCrane.Lobber.Die\")"
            (:then
              (Ability|Async|WaitGameplayEventToActor 0 "(TagName=\"Event.TidalCrane.Lobber.Attack\")"
                (:then
                  (MusicEncounterComponent|SetMusicEncounterByInstanceTag (Variables|Default|GetBP_MusicEncounterComponent) (Variables|Default|GetMusic%20Encounter%20Instance%20Tag)))
                (:EventReceived
                  (CallFunction|OnAttack))))
            (:EventReceived
              (CallFunction|OnDeathEvent))))))
    (:then_1
      ; Spawn Audio and Dirt VFX
      (Networking|CanExecuteCosmeticEvents
        (:True
          (CallFunction|ForceUpdateCustomDepth)
          (Utilities|Time|SetTimerbyEvent (delegate Custom|UpdateCustomDepth) (Variables|Default|GetCustom%20Depth%20Update%20Rate) true)
          (Utilities|FlowControl|Sequence
            (:then_0
              (if (not (Variables|Default|GetInitialSpawnVFXDone))
                (Audio|PlaySoundatLocation "/ProtoSwarm/Assets/Audio/TidalCrane_Brute/MSS_TidalCrane_Brute_Spawn_Tell.MSS_TidalCrane_Brute_Spawn_Tell" _returnvalue)
                (Niagara|SpawnSystematLocation (Variables|Default|GetSpawnInVFX) _returnvalue (Transformation|GetActorRotation) "1.000000,1.000000,1.000000" true true "None" false)))
            (:then_1
              (CallFunction|StartUpdatingHleathBarWidget)
              (Health|AssignOnPawnHealthChanged (EventDispatchers|CreateEvent :!event Custom|PawnHealthHasChanged))
              (Rendering|Material|SetCustomPrimitiveDataFloat (Variables|Character|GetMesh) 2 0.0) ; Tech art
              (bind _returnvalue_2 (Rendering|Material|CreateDynamicMaterialInstance (Variables|Default|GetMaskedDeathMaterial) "None" "None" :!class /Script/Engine.KismetMaterialLibrary))
              (Variables|Default|SetMID_MaskedDeathMaterial _returnvalue_2)
              (Utilities|FlowControl|Delay 1.0)
              (Rendering|SetActorHiddenInGame false))))))))

(event Custom|UpdateCustomDepth
  (bind _fortpawncomponent_customdepth (Variables|Default|GetFortPawnComponent_CustomDepth))
  ; Spawn Audio and Dirt VFX
  (if (CustomDepth|IsRenderCustomDepthEnabled _fortpawncomponent_customdepth)
    (CustomDepth|Update _fortpawncomponent_customdepth (Variables|Default|GetCustom%20Depth%20Update%20Rate) false)))

(event Custom|SetupProximityComponent
  (bind _fortnearbyactorsperception (Variables|Default|GetFortNearbyActorsPerception))
  ; Look to see if there are any threats nearby
  (Utilities|FlowControl|DelayUntilNextTick)
  (Variables|InternalData|SetAIBlackboard (AI|GetBlackboard (Pawn|GetController)))
  (Components|Activation|Activate _fortnearbyactorsperception false)
  (bind _returnvalue (Query|BPRegisterListener _fortnearbyactorsperception (CallFunction|GetPerceptionConfig) (EventDispatchers|CreateEvent :!event Custom|ThreatsInProximity)))
  (Variables|Default|SetPerceptionListener _returnvalue))

(event Custom|ThreatsInProximity
  (params (AllFoundActors :type Array%20of%20Actor%20Object%20References) (EnteringActors :type Array%20of%20Actor%20Object%20References) (ExitingActors :type Array%20of%20Actor%20Object%20References))
  ; Look to see if there are any threats nearby
  (Variables|InternalData|SetPawnsInProximity AllFoundActors)
  (if (not (GameplayTags|HasMatchingGameplayTag "(TagName=\"Athena.Gameplay.Status.IsControlledByTakeOverAIPawn\")"))
    (CallFunction|CheckforValidThreat)))

(event Custom|PawnHealthHasChanged
  ;[1 1 1] Health Bar Logic
  (CallFunction|TryShowHealthBar))

(event Custom|StartUpdatingHleathBarWidget
  ;[1 1 1] Health Bar Logic
  (bind _npcstatuswidget (Variables|Default|GetNPCStatusWidget)
    (:then
      (bind _asnpcstatus_widget (Utilities|Casting|CastToNPCStatusWidget (UserInterface|GetUserWidgetObject _npcstatuswidget)))
      (Utilities|FlowControl|Sequence
        (:then_0
          (Components|Tick|SetComponentTickEnabled (Variables|Default|GetNPCStatusWidget) (Widget|IsVisible _asnpcstatus_widget)))
        (:then_1
          (Default|BindEventtoVisibilityChanged _asnpcstatus_widget (delegate Custom|WidgetVisibilityChanged)))))
    (:else)))

(event Custom|WidgetVisibilityChanged
  (params (bVisible :type Boolean))
  ;[1 1 1] Health Bar Logic
  (Components|Tick|SetComponentTickEnabled (Variables|Default|GetNPCStatusWidget) bVisible))

(event GameplayCue|GameplayCue.Impact.Physical.Creature
  (params (EventType :type EGameplayCueEvent%20Enum :default "OnActive") (Parameters :type Gameplay%20Cue%20Parameters%20Structure%20%28by%20ref%29))
  (bind (_bblockinghit _binitialoverlap _time _distance _location) (Collision|BreakHitResult (Ability|GameplayCue|GetHitResult Parameters)))
  ;[1 1 1] ImpactATarget
  (Audio|PlaySoundatLocation (Sound|GetImpactPhysicalSurfaceSound (Ability|GameplayCue|GetHitResult Parameters)) _location)
  (Default|PlayDamageFXSurfaceEnemies Parameters (Variables|Default|GetImpactParticles_Default) (Variables|Default|GetImpactParticles_Wood) (Variables|Default|GetImpactParticles_Stone) (Variables|Default|GetImpactParticles_Metal) (Variables|Default|GetImpactParticles_HumanFlesh) (.location (Transformation|GetSocketTransform (Variables|Character|GetMesh) (Variables|Default|GetImpact_MeleeSocket) "RTS_World")) (.rotation (Transformation|GetSocketTransform (Variables|Character|GetMesh) (Variables|Default|GetImpact_MeleeSocket) "RTS_World")) (Variables|Default|GetMeleeImpact_UseSocketTransforms)))

(event Custom|StopUpdatingHealthBarWidget
  ;[1 1 1] Health Bar Logic
  (bind _npcstatuswidget (Variables|Default|GetNPCStatusWidget)
    (:then
      (bind _asnpcstatus_widget (Utilities|Casting|CastToNPCStatusWidget (UserInterface|GetUserWidgetObject _npcstatuswidget)))
      (Default|UnbindEventfromVisibilityChanged _asnpcstatus_widget (delegate Custom|WidgetVisibilityChanged)))
    (:else)))

(event Custom|NPC%20UpdateHealthBarComponentVisibility
  (bind _returnvalue (Utilities|Time|GetGameTimeinSeconds))
  (bind _npc_death_time (Variables|InternalData|GetNPC%20Death%20Time))
  (bind _returnvalue_2 (- _returnvalue _npc_death_time))
  ;[1 1 1] Health Bar Logic
  (if (> (Variables|InternalData|GetNPC%20Death%20Time) 0)
    (bind (_asbool _asintrounded _asfloat) (CallFunction|NPCGetScalableFloatHotfixValue (Variables|Default|HealthBar|GetNPC%20HealthBarHideTimeAfterDeathHF)))
    (if (and (> _asfloat 0.0) (not (Variables|Default|HealthBar|GetNPC%20StatusWidgetHealthBarDisabled)))
      (bind _npcstatuswidget (Variables|Default|GetNPCStatusWidget)
        (:then
          (bind _asnpcstatus_widget (Utilities|Casting|CastToNPCStatusWidget (UserInterface|GetUserWidgetObject _npcstatuswidget)))
          (Class|NPCStatusWidget|EventUpdateStatus _asnpcstatus_widget 0.0 0.0 false :!pinned true) ; Make sure  Health Bar is 0
          (bind (_asbool_2 _asintrounded_2 _asfloat_2) (CallFunction|NPCGetScalableFloatHotfixValue (Variables|Default|HealthBar|GetNPC%20HealthBarHideTimeAfterDeathHF)))
          (if (> (- _asfloat_2 _returnvalue_2) 0)
            (Utilities|FlowControl|Delay (- _asfloat_2 _returnvalue_2))
            (else
              _))
          (bind _npcstatuswidget_2 (Variables|Default|GetNPCStatusWidget)
            (:then
              (Development|SetHiddeninGame _npcstatuswidget_2 true false :!pinned true) ; Hide health bar (probably after animation)
            )
            (:else)))
        (:else))
      (else
        _))))

(event WaterBody|EventOnBeginDeepWaterInteraction
  (params (WaterSurfaceLocation :type Vector :default "0, 0, 0") (WaterDepth :type Float%20%28single-precision%29) (WaterBody :type Water%20Body%20Object%20Reference))
  ; Water
  (CallFunction|OnDeepWaterEnter))

(event Custom|OnDeathEvent
  (bind _mesh (Variables|Character|GetMesh))
  ; Death
  (if (ScalableFloat|GetValueasBool (Variables|Default|GetPlayBurstVFXOnElim))
    (bind _returnvalue (Niagara|SpawnSystematLocation (Variables|Default|GetFX_BugDeath) (+ (Transformation|GetActorLocation) "0.000000,0.000000,-100.000000") (Transformation|GetActorRotation)))
    (Niagara|SetNiagaraVariable(Float) _returnvalue "FXScale" 4.0)
    (Niagara|SetNiagaraVariable(Bool) _returnvalue "FXDecal" false)
    (else
      _))
  (if (ScalableFloat|GetValueasBool (Variables|Default|GetPlayFadeVFXOnElim))
    (Rendering|Material|SetMaterial _mesh 0 (Variables|Default|GetMID_MaskedDeathMaterial))
    (Rendering|Material|SetCustomPrimitiveDataFloat _mesh 2 1.0)
    (Rendering|Material|SetCustomPrimitiveDataFloat _mesh 3 (Utilities|Time|GetGameTimeinSeconds))
    (Utilities|FlowControl|Delay 2.0)
    (else
      _))
  (FortPawn|HideBodyonDeath false))

(event Custom|OnAttack
  ; Attack
  (Rendering|Material|SetCustomPrimitiveDataFloat (Variables|Character|GetMesh) 5 (Utilities|Time|GetGameTimeinSeconds)))

(event MusicEncounter|EventSetMusicEncounterInstanceTag
  (params (Music%20Encounter%20Instance%20Tag :type Gameplay%20Tag%20Structure))
  (Variables|Default|SetMusic%20Encounter%20Instance%20Tag Music%20Encounter%20Instance%20Tag))

(event Custom|DoWeakpointHitFX
  (params (Location :type Vector) (Normal :type Vector) (Bone%20Name :type Name))
  ;[0.15 0.05469 0.05469 0.5] VFX: Weakpoint FX
  (Networking|CanExecuteCosmeticEvents
    (:True
      (if (== Bone%20Name "pod_center")
        (Niagara|SpawnSystematLocation "/ProtoSwarm/Assets/FX/NS_Weakspot_Impact_Brute.NS_Weakspot_Impact_Brute" Location (Math|Conversions|RotationFromXVector (Math|Vector|Normalize (+ Normal (Utilities|RandomVectorinRange "-0.110000,-0.110000,-0.110000" "0.110000,0.110000,0.110000")))))))))

(event EventOnDamageServer
  (params (Damage :type Float%20%28single-precision%29) (DamageTags :type Gameplay%20Tag%20Container%20Structure%20%28by%20ref%29) (Momentum :type Vector :default "0, 0, 0") (HitInfo :type Hit%20Result%20Structure%20%28by%20ref%29) (InstigatedBy :type Controller%20Object%20Reference) (DamageCauser :type Actor%20Object%20Reference) (EffectContext :type Gameplay%20Effect%20Context%20Handle%20Structure))
  (bind (_bblockinghit _binitialoverlap _time _distance _location _impactpoint _normal _impactnormal _physmat _hitactor _hitcomponent _hitbonename) (Collision|BreakHitResult HitInfo))
  ;[0.15 0.05469 0.05469 0.5] VFX: Weakpoint FX
  (CallFunction|DoWeakpointHitFX _impactpoint _impactnormal _hitbonename))

(fn ApplyBuildingHitEffectToSkeletalMesh
  (params (Mesh :type Skeletal%20Mesh%20Component%20Object%20Reference))
  (bind _buildinghitfxskeletalmeshes (Variables|Default|GetBuildingHitFXSkeletalMeshes))
  (bind (_value _returnvalue) (Utilities|Map|Find _buildinghitfxskeletalmeshes Mesh))
  (bind _fxskeletalmesh (LocalVariables|GetFXSkeletalMesh))
  (Utilities|IsValid Mesh
    (:Is%20Valid
      (if _returnvalue
        (bind _output_get (LocalVariables|SetFXSkeletalMesh _value))
        (Utilities|IsValid _output_get
          (:Is%20Valid
            _)
          (:Is%20Not%20Valid
            (bind _returnvalue_2 (AddComponent|NODE_AddFXSkeletonMeshComponent-0))
            (Utilities|Map|Add (Variables|Default|GetBuildingHitFXSkeletalMeshes) Mesh _returnvalue_2)
            (bind _output_get_2 (LocalVariables|SetFXSkeletalMesh _returnvalue_2))
            (Class|FortFXSkeletonMeshComponent|SetSource _output_get_2 Mesh)
            (Variables|/Script/FortniteGame.FortFXSkeletonMeshComponent|SetBuildingHitDuration (Variables|Default|GetBuildingHitDuration) _fxskeletalmesh)))
        (Class|FXSkeletonMeshComponent|ApplyBuildingHitEffect _fxskeletalmesh)
        (else
          _)))))

(fn ApplyBuildingHitEffectToStaticMesh
  (params (Mesh :type Static%20Mesh%20Component%20Object%20Reference))
  (bind _buildinghitfxstaticmeshes (Variables|Default|GetBuildingHitFXStaticMeshes))
  (bind (_value _returnvalue) (Utilities|Map|Find _buildinghitfxstaticmeshes Mesh))
  (bind _fxstaticmesh (LocalVariables|GetFXStaticMesh))
  (Utilities|IsValid Mesh
    (:Is%20Valid
      (if _returnvalue
        (bind _output_get (LocalVariables|SetFXStaticMesh _value))
        (Utilities|IsValid _output_get
          (:Is%20Valid
            _)
          (:Is%20Not%20Valid
            (bind _returnvalue_2 (AddComponent|NODE_AddFXStaticMeshComponent-2))
            (Utilities|Map|Add (Variables|Default|GetBuildingHitFXStaticMeshes) Mesh _returnvalue_2)
            (bind _output_get_2 (LocalVariables|SetFXStaticMesh _returnvalue_2))
            (Class|FortFXStaticMeshComponent|SetSource _output_get_2 Mesh)
            (Variables|/Script/FortniteGame.FortFXStaticMeshComponent|SetBuildingHitDuration (Variables|Default|GetBuildingHitDuration) _fxstaticmesh)))
        (Class|FXStaticMeshComponent|ApplyBuildingHitEffect _fxstaticmesh)
        (else
          _)))))

(fn CheckForValidThreat
  (AI|Components|Blackboard|SetValueasBool (Variables|InternalData|GetAIBlackboard) (Variables|Default|Perception|GetKey_ThreatsInProximity) (> (Utilities|Array|Length (Variables|InternalData|GetPawnsInProximity)) 0)))

(fn GetLocationWithCapsuleHalfHeight
  (return (Utilities|Operators|Subtract (Transformation|GetActorLocation) (Math|Vector|MakeVector 0.0 0.0 (Components|Capsule|GetScaledCapsuleHalfHeight (Variables|Character|GetCapsuleComponent))))))

(fn GetPerceptionConfig
  (return (Utilities|Struct|MakeFortNearbyActorsPerceptionConfiguration 1.0 7 0 (ScalableFloat|GetValueatLevel (Variables|Default|Perception|GetPerceptionDistance)) "Friendly" true (Variables|Default|Perception|GetPerception_TagQuery))))

(fn NPCGetScalableFloatHotfixValue
  (params (ScalableFloatHotfix :type Scalable%20Float%20Structure) (AtLevel :type Float%20%28double-precision%29))
  (if (CallFunction|NPCScalableFloatIsValid ScalableFloatHotfix)
    (return (> (Math|Float|Round (ScalableFloat|GetValueatLevel ScalableFloatHotfix AtLevel)) 0) (Math|Float|Round (ScalableFloat|GetValueatLevel ScalableFloatHotfix AtLevel)) (ScalableFloat|GetValueatLevel ScalableFloatHotfix AtLevel))
    (else
      (return (> (Math|Float|Round (Utilities|Struct|BreakScalableFloat ScalableFloatHotfix)) 0) (Math|Float|Round (Utilities|Struct|BreakScalableFloat ScalableFloatHotfix)) (Utilities|Struct|BreakScalableFloat ScalableFloatHotfix)))))

(fn NPCScalableFloatIsValid
  (params (ScalableFloat :type Scalable%20Float%20Structure))
  (bind (_value _curve) (Utilities|Struct|BreakScalableFloat ScalableFloat))
  (bind (_curvetable _rowname) (Utilities|Struct|BreakCurveTableRowHandle _curve))
  (return (and (Utilities|IsValid _curvetable :!class /Script/Engine.KismetSystemLibrary) (Utilities|Name|NotEqual(Name) _rowname "None"))))

(fn OnDeepWaterEnter
  (Networking|SwitchHasAuthority
    (:Authority
      (Abilities|TryActivateAbilitiesbyTag (Ability|GetFortAbilitySystemComponent self) (Variables|Default|GetDrown%20Ability%20Gameplay%20Tag) true))))

(fn OnRep_MusicEncounterInstanceTag
  (Networking|CanExecuteCosmeticEvents
    (:True
      (MusicEncounterComponent|SetMusicEncounterByInstanceTag (Variables|Default|GetBP_MusicEncounterComponent) (Variables|Default|GetMusic%20Encounter%20Instance%20Tag)))))

(fn OnRep_NPCDeathTime
  (Networking|CanExecuteCosmeticEvents
    (:True
      (CallFunction|NPCUpdateHealthBarComponentVisibility))))

(fn PlayAdditiveHitReacts
  (params (Hit%20Direction :type Vector) (Anim%20Montage :type Anim%20Montage%20Object%20Reference))
  (HitReacts|PlayAdditiveHitReactsNative Hit%20Direction Anim%20Montage) ; This used to be done directly in the Event graph, which meant any Child of this Pawn couldn't override the Function behavior.\n\nWith this, we should be able to override PlayAdditiveHitReacts and call the Native function here with different logic in front of it, if we need to.
)

(fn PlaySoundLibraryAudio
  (params (Event%20Name :type Gameplay%20Tag%20Structure))
  (Networking|CanExecuteCosmeticEvents
    (:True
      (if (Utilities|IsValid (Variables|Default|GetSoundLibraryContext) :!class /Script/Engine.KismetSystemLibrary)
        (Utilities|Struct|SetmembersinSimpleSoundSettings (Variables|/Script/SoundLibrary.SoundLibrarySimpleContext|GetSettings (Variables|Default|GetSoundLibraryContext)) Event%20Name)
        (else
          (bind _returnvalue (SoundLibrary|CreateSoundLibraryContext :Class "/Script/SoundLibrary.SoundLibrarySimpleContext" :Settings (Utilities|Struct|MakeSimpleSoundSettings Event%20Name) :Actor self :bForwardToActorOwner true :bForwardToAttachedActors true :bAutoResetContext true))
          (bind _output_get (Variables|Default|SetSoundLibraryContext _returnvalue))))
      (SoundLibrary|PlaySoundwithContext (WorldSubsystems|GetSoundLibrarySubsystem) _output_get))))

(fn SetScalarParameterOnAllCharacterMIDs
  (params (Parameter%20Name :type Name) (Scalar%20Value :type Float%20%28double-precision%29) (AlsoModifyOriginalMIDs :type Boolean))
  (bind _auxillary_objects (Variables|Default|GetAuxillary%20Objects))
  (bind _duplicate_auxillary_objects (Variables|Default|GetDuplicate%20Auxillary%20Objects))
  (Utilities|FlowControl|Sequence
    (:then_0
      (bind _mesh (Variables|Character|GetMesh)
        (:then
          (Rendering|Material|SetScalarParameterValueonMaterials _mesh (Variables|GetParameter%20Name) (Variables|GetScalar%20Value)))
        (:else
          _))
      (bind _duplicatecharactermesh (Variables|Default|GetDuplicateCharacterMesh)
        (:then
          (Rendering|Material|SetScalarParameterValueonMaterials _duplicatecharactermesh (Variables|GetParameter%20Name) (Variables|GetScalar%20Value)))
        (:else)))
    (:then_1
      (if (Utilities|IsArrayNonEmpty _auxillary_objects)
        (for _array_element _auxillary_objects
          (Utilities|IsValid _array_element
            (:Is%20Valid
              (bind _asmesh_component (Utilities|Casting|CastToMeshComponent _array_element))
              (Rendering|Material|SetScalarParameterValueonMaterials _asmesh_component (Variables|GetParameter%20Name) (Variables|GetScalar%20Value)))))))
    (:then_2
      (if (Utilities|IsArrayNonEmpty _duplicate_auxillary_objects)
        (for _array_element_2 _duplicate_auxillary_objects
          (Utilities|IsValid _array_element_2
            (:Is%20Valid
              (bind _asmesh_component_2 (Utilities|Casting|CastToMeshComponent _array_element_2))
              (Rendering|Material|SetScalarParameterValueonMaterials _asmesh_component_2 (Variables|GetParameter%20Name) (Variables|GetScalar%20Value)))))))
    (:then_3
      ;[1 1 1] Optionally keep the original MIDs in sync with the override MIDs
      (if AlsoModifyOriginalMIDs
        (Utilities|FlowControl|Sequence
          (:then_0
            (for _array_element_3 (Variables|Default|GetPrevious%20MID)
              (Rendering|Material|SetScalarParameterValue _array_element_3 (Variables|GetParameter%20Name) (Variables|GetScalar%20Value) :!class /Script/Engine.MaterialInstanceDynamic)))
          (:then_1
            (for _array_element_4 (Variables|Default|GetPrevious%20MID_AuxObjs)
              (Rendering|Material|SetScalarParameterValue _array_element_4 (Variables|GetParameter%20Name) (Variables|GetScalar%20Value) :!class /Script/Engine.MaterialInstanceDynamic))))))))

(fn SetupStatusWidget
  (bind _npcstatuswidget (Variables|Default|GetNPCStatusWidget)
    (:then
      (bind _asnpcstatus_widget (Utilities|Casting|CastToNPCStatusWidget (UserInterface|GetUserWidgetObject _npcstatuswidget)))
      (Variables|/Game/Athena/HUD/Actors/NPCStatusWidget.NPCStatusWidget_C|SetFortPawn self _asnpcstatus_widget))
    (:else)))

(fn SetVectorParameterOnAllCharacterMIDs
  (params (Parameter%20Name :type Name) (Vector%20Value :type Vector) (Propagate%20to%20Auxiliary%20Meshes :type Boolean :default true))
  (bind _auxillary_objects (Variables|Default|GetAuxillary%20Objects))
  (bind _duplicate_auxillary_objects (Variables|Default|GetDuplicate%20Auxillary%20Objects))
  (bind _mesh (Variables|Character|GetMesh)
    (:then
      (Rendering|Material|SetVectorParameterValueonMaterials _mesh Parameter%20Name Vector%20Value))
    (:else
      _))
  (bind _duplicatecharactermesh (Variables|Default|GetDuplicateCharacterMesh)
    (:then
      (Rendering|Material|SetVectorParameterValueonMaterials _duplicatecharactermesh Parameter%20Name Vector%20Value))
    (:else
      _))
  (if (and (> (Utilities|Array|Length _auxillary_objects) 0) (Variables|GetPropagate%20to%20Auxiliary%20Meshes))
    (for _array_element _auxillary_objects
      (Utilities|IsValid _array_element
        (:Is%20Valid
          (bind _asmesh_component (Utilities|Casting|CastToMeshComponent _array_element))
          (Rendering|Material|SetVectorParameterValueonMaterials _asmesh_component (Variables|GetParameter%20Name) (Variables|GetVector%20Value)))))
    (if (and (> (Utilities|Array|Length _duplicate_auxillary_objects) 0) (Variables|GetPropagate%20to%20Auxiliary%20Meshes))
      (for _array_element_2 _duplicate_auxillary_objects
        (Utilities|IsValid _array_element_2
          (:Is%20Valid
            (bind _asmesh_component_2 (Utilities|Casting|CastToMeshComponent _array_element_2))
            (Rendering|Material|SetVectorParameterValueonMaterials _asmesh_component_2 (Variables|GetParameter%20Name) (Variables|GetVector%20Value)))))))
  (return))

(fn ShouldUpdateSkeletalMesh
  ;[0.15 0 0 0.5] HACK - Without this, PlayMontagesAndAnimNotifies will NOT trigger in GameplayAbilities!!
  (return true))

(fn TryShowHealthBar
  (bind _array_element (Utilities|Array|ForEachLoopwithBreak (Fortnite|GetLocalFortPlayerControllers true))
    (:LoopBody
      (bind (_asbool _asintrounded _asfloat) (CallFunction|NPCGetScalableFloatHotfixValue (Variables|Default|HealthBar|GetNPC%20MaxDistanceFromDamagedNPCToDrawHeallthBarHF)))
      (if (and (< (Class|AbilityFunctions|GetDistanceBetweenActors (Pawn|GetControlledPawn _array_element) self) _asfloat) (not (Variables|Default|HealthBar|GetNPC%20StatusWidgetHealthBarDisabled)))
        (bind _npcstatuswidget (Variables|Default|GetNPCStatusWidget)
          (:then
            (bind _asnpcstatus_widget (Utilities|Casting|CastToNPCStatusWidget (UserInterface|GetUserWidgetObject _npcstatuswidget)))
            (Class|NPCStatusWidget|EventUpdateStatus _asnpcstatus_widget (/ (FortPawn|GetHealth) (FortPawn|GetMaxHealth)) 0.0 false))
          (:else))))
    (:Completed
      _))
  (return))

(fn ConstructionScript
  (|CallParentFunction))
