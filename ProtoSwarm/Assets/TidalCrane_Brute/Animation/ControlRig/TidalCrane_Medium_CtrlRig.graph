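# Control Rig for the TidalCrane Medium character that tilts and lowers the root bone to match the ground slope.
# A downward sphere trace from the root yields spring-smoothed hit location and normal, reusing previous values on misses and resetting on large height jumps.
# Root Alignment converts the normal into two tilt angles, crouches with slope, and springs root translation.
# With Debug enabled, a spawned Slope control's plane replaces the trace.
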ControlRigBlueprint /ProtoSwarm/Assets/TidalCrane_Brute/Animation/ControlRig/TidalCrane_Medium_CtrlRig
graphs [
 New Function_ContainedGraph {nodes [
   RigVMFunctionEntryNode()
   > RigVMFunction_Sequence_1 = RigVMAggregateNode()
   RigVMFunction_ControlFlowBranch = &RigVMFunction_ControlFlowBranch(Condition=false, ExecuteContext=@RigVMFunction_Sequence_1.A)
   n1 = /Script/ControlRig.RigUnit_SphereTraceByTraceChannel(
    Radius=5
    TraceChannel=TraceTypeQuery10
    End=/Script/RigVM.RigVMFunction_MathVectorAdd(B=(X=0,Y=0,Z=-500), A=RigVMVariableNode() {variable=Start Location}.Value) {is=struct RigVMUnitNode, template="Add::Execute(in A,in B,out Result)"}.Result
    Start=RigVMVariableNode() {variable=Start Location}.Value
   ) {is=struct RigVMUnitNode}
   n2 = &RigVMFunction_ControlFlowBranch(ExecuteContext=@RigVMFunction_ControlFlowBranch.False, Condition=@n1.bHit)
   RigVMVariableNode(ExecuteContext=@n2.False) {variable=GroundTraceHit}
   > RigVMVariableNode(Value=RigVMVariableNode() {variable=Previous Location}.Value) {variable=HitLocation}
   > RigVMVariableNode(Value=RigVMVariableNode() {variable=Previous Normal}.Value) {variable=HitNormal}
   RigVMVariableNode(Value=true, ExecuteContext=@n2.True) {variable=GroundTraceHit}
   SpringInterp = /Script/ControlRig.RigUnit_SpringInterpVectorV2(
    CriticalDamping=1
    Force=(X=0,Y=0,Z=0)
    Strength=4
    TargetVelocityAmount=0
    bInitializeFromTarget=False
    bUseCurrentInput=False
    Target=@n1.HitLocation
    Current=RigVMVariableNode() {variable=Previous Location}.Value
   ) {
    is=struct RigVMUnitNode
    template="SpringInterp::Execute(in Target,in Strength,in CriticalDamping,in Force,in bUseCurrentInput,in Current,in TargetVelocityAmount,in bInitializeFromTarget,out Result,out Velocity)"
   }
   > n3 = RigVMVariableNode(Value=@SpringInterp.Result) {variable=HitLocation}
   n4 = &RigVMFunction_ControlFlowBranch(
    Condition=/Script/RigVM.RigVMFunction_MathVectorIsNearlyZero(Tolerance=0.001, Value=@n1.HitNormal) {is=struct RigVMUnitNode, template="IsNearlyZero::Execute(in Value,in Tolerance,out Result)"}.Result
    ExecuteContext=@n3.ExecuteContext
   )
   n5 = &RigUnit_SpringInterpVectorV2(
    CriticalDamping=1
    Force=(X=0,Y=0,Z=0)
    Strength=4
    TargetVelocityAmount=0
    bInitializeFromTarget=False
    bUseCurrentInput=False
    Current=RigVMVariableNode() {variable=Previous Normal}.Value
    Target=@n1.HitNormal
   ) {template="SpringInterp::Execute(in Target,in Strength,in CriticalDamping,in Force,in bUseCurrentInput,in Current,in TargetVelocityAmount,in bInitializeFromTarget,out Result,out Velocity)"}
   Unit = /Script/RigVM.RigVMFunction_MathVectorUnit(Value=@n5.Result) {is=struct RigVMUnitNode, template="Unit::Execute(in Value,out Result)"}
   RigVMVariableNode(ExecuteContext=@n4.False, Value=@Unit.Result) {variable=HitNormal}
   RigVMVariableNode(Value=(X=0,Y=0,Z=1), ExecuteContext=@n4.True) {variable=HitNormal}
   n6 = &RigUnit_GetTransform(bInitial=False, Item=RigVMVariableNode() {variable=Control}.Value)
   n7 = /Script/RigVM.RigVMFunction_MathQuaternionRotateVector(Vector=(X=0,Y=0,Z=1), Transform=@n6.Transform.Rotation) {is=struct RigVMUnitNode, template="Rotate Vector::Execute(in Transform,in Vector,out Result)"}
   RigVMFunction_MathIntersectPlane = /Script/RigVM.RigVMFunction_MathIntersectPlane(
    Direction=(X=0,Y=0,Z=-1)
    PlanePoint=@n6.Transform.Translation
    PlaneNormal=@n7.Result
    Start=RigVMVariableNode() {variable=Start Location}.Value
   ) {is=struct RigVMUnitNode}
   n8 = &RigUnit_SpringInterpVectorV2(
    CriticalDamping=1
    Force=(X=0,Y=0,Z=0)
    Strength=4
    TargetVelocityAmount=0
    bInitializeFromTarget=False
    bUseCurrentInput=False
    Target=@RigVMFunction_MathIntersectPlane.Result
    Current=RigVMVariableNode() {variable=Previous Location}.Value
   ) {template="SpringInterp::Execute(in Target,in Strength,in CriticalDamping,in Force,in bUseCurrentInput,in Current,in TargetVelocityAmount,in bInitializeFromTarget,out Result,out Velocity)"}
   RigVMVariableNode(Value=@n8.Result, ExecuteContext=@RigVMFunction_ControlFlowBranch.True) {variable=HitLocation}
   n9 = &RigUnit_SpringInterpVectorV2(
    CriticalDamping=1
    Force=(X=0,Y=0,Z=0)
    Strength=4
    TargetVelocityAmount=0
    bInitializeFromTarget=False
    bUseCurrentInput=False
    Target=@n7.Result
    Current=RigVMVariableNode() {variable=Previous Normal}.Value
   ) {template="SpringInterp::Execute(in Target,in Strength,in CriticalDamping,in Force,in bUseCurrentInput,in Current,in TargetVelocityAmount,in bInitializeFromTarget,out Result,out Velocity)"}
   > RigVMVariableNode(Value=@n9.Result) {variable=HitNormal}
   > Draw Trace_1 = RigVMFunctionReferenceNode(
    Color=(R=1,G=0,B=0,A=1)
    Factor=20
    Radius=25
    Hit Location=@RigVMFunction_MathIntersectPlane.Result
    Hit Normal=@n9.Result
    Start Position=RigVMVariableNode() {variable=Start Location}.Value
   )
   RigVMFunctionReturnNode(
    ExecuteContext=@RigVMFunction_Sequence_1.C
    Hit Normal=RigVMVariableNode() {variable=HitNormal}.Value
    Hit Location=RigVMVariableNode() {variable=HitLocation}.Value
   )
   n10 = RigVMVariableNode() {variable=Previous Location}
   DISPATCH_RigVMDispatch_CoreNotEquals = RigVMDispatchNode(B=(X=0,Y=0,Z=0), A=@n10.Value) {template="DISPATCH_RigVMDispatch_CoreNotEquals(in A,in B,out Result)"}
   Subtract = &RigVMFunction_MathDoubleSub(A=RigVMVariableNode() {variable=HitLocation}.Value.Z, B=@n10.Value.Z)
   Absolute = &RigVMFunction_MathDoubleAbs(Value=@Subtract.Result)
   Greater = /Script/RigVM.RigVMFunction_MathDoubleGreater(A=@Absolute.Result, B=RigVMVariableNode() {variable=Max Distance}.Value) {is=struct RigVMUnitNode, template="Greater::Execute(in A,in B,out Result)"}
   RigVMFunction_MathBoolAnd = &RigVMFunction_MathBoolAnd(B=@DISPATCH_RigVMDispatch_CoreNotEquals.Result, A=@Greater.Result)
   n11 = &RigVMFunction_ControlFlowBranch(Condition=@RigVMFunction_MathBoolAnd.Result, ExecuteContext=@RigVMFunction_Sequence_1.B)
   RigVMVariableNode(Value=(X=0,Y=0,Z=0), ExecuteContext=@n11.True) {variable=HitLocation}
   > RigVMVariableNode(Value=(X=0,Y=0,Z=1)) {variable=HitNormal}
   RigVMFunction_DebugTransformMutableNoSpace = /Script/RigVM.RigVMFunction_DebugTransformMutableNoSpace(
    Color=(R=1,G=0,B=0,A=1)
    DebugDrawSettings=(DepthPriority=SDPG_Foreground,Lifetime=-1)
    ExecutePin=()
    Mode=Axes
    Scale=10
    Thickness=0
    Transform=(Rotation=(X=0,Y=0,Z=0,W=1),Translation=(X=0,Y=0,Z=0),Scale3D=(X=1,Y=1,Z=1))
    WorldOffset=(Rotation=(X=0,Y=0,Z=0,W=1),Translation=(X=0,Y=0,Z=0),Scale3D=(X=1,Y=1,Z=1))
    bEnabled=True
   ) {is=struct RigVMUnitNode}
   RigVMVariableNode() {variable=Use Control}
  ]}
 New Function_ContainedGraph {nodes [
   RigVMFunctionEntryNode()
   n1 = &RigUnit_GetTransform(bInitial=False, Item=RigVMVariableNode() {variable=Bone}.Value)
   DISPATCH_RigVMDispatch_MakeStruct = RigVMDispatchNode(
    Elements=(X=0,Y=0,Z=0)
    Elements.X=@n1.Transform.Translation.X
    Elements.Y=@n1.Transform.Translation.Y
    Elements.Z=RigVMVariableNode() {variable=TraceHeight}.Value
   ) {template="DISPATCH_RigVMDispatch_MakeStruct(in Elements,out Struct)"}
   n2 = RigVMVariableNode() {variable=Hit Location}
   n3 = RigVMVariableNode() {variable=Hit Normal}
   > Trace From Location = RigVMFunctionReferenceNode(
    Max Distance=100
    Start Location=@DISPATCH_RigVMDispatch_MakeStruct.Struct
    Previous Location=@n2.Value
    Previous Normal=@n3.Value
    Control=RigVMVariableNode() {variable=Debug Control}.Value
    Use Control=RigVMVariableNode() {variable=Debug}.Value
   )
   > RigVMFunctionReturnNode(Hit Location Out=@"Trace From Location".Hit Location, Hit Norma Out=@"Trace From Location".Hit Normal)
   Draw Trace = RigVMFunctionReferenceNode(Color=(R=1,G=0,B=0,A=1), ExecuteContext=(), Factor=20, Radius=25, Start Position=(X=0,Y=0,Z=0), Hit Location=@n2.Value, Hit Normal=@n3.Value)
  ]}
 New Function_ContainedGraph {nodes [
   RigVMFunctionEntryNode()
   Slope Angles Radians_1 = RigVMFunctionReferenceNode(Normal=RigVMVariableNode() {variable=Hit Normal}.Value)
   Negate = /Script/RigVM.RigVMFunction_MathDoubleNegate(Value=@"Slope Angles Radians_1".X Angle) {is=struct RigVMUnitNode, template="Negate::Execute(in Value,out Result)"}
   n1 = &RigVMFunction_MathDoubleMul(A=&RigVMFunction_MathDoubleAbs(Value=@Negate.Result).Result, B=RigVMVariableNode() {variable=Crouching Amount}.Value)
   RigUnit_GetTransform = &RigUnit_GetTransform(bInitial=False, Item=RigVMVariableNode() {variable=Item}.Value)
   n2 = /Script/RigVM.RigVMFunction_MathTransformMakeAbsolute(
    Local=(Rotation=(X=0,Y=0,Z=0,W=1),Translation=(X=0,Y=0,Z=0),Scale3D=(X=1,Y=1,Z=1))
    Parent=(Rotation=(X=0,Y=0,Z=0,W=1),Translation=(X=0,Y=0,Z=0),Scale3D=(X=1,Y=1,Z=1))
    Local.Translation.Z=&RigVMFunction_MathDoubleAdd(B=@n1.Result, A=@RigUnit_GetTransform.Transform.Translation.Z).Result
    Local.Rotation=@RigUnit_GetTransform.Transform.Rotation
    Local.Translation.X=@RigUnit_GetTransform.Transform.Translation.X
    Local.Translation.Y=@RigUnit_GetTransform.Transform.Translation.Y
    Parent.Rotation=/Script/RigVM.RigVMFunction_MathQuaternionFromAxisAndAngle(Axis=(X=1,Y=0,Z=0), Angle=@Negate.Result) {is=struct RigVMUnitNode, template="FromAxisAndAngle::Execute(in Axis,in Angle,out Result)"}.Result
   ) {is=struct RigVMUnitNode, template="Make Absolute::Execute(in Local,in Parent,out Global)"}
   n3 = RigVMVariableNode() {variable=Side Rotation Factor}
   n4 = &RigVMFunction_MathTransformMakeAbsolute(
    Parent=(Rotation=(X=0,Y=0,Z=0,W=1),Translation=(X=0,Y=0,Z=0),Scale3D=(X=1,Y=1,Z=1))
    Local=@n2.Global
    Parent.Rotation=&RigVMFunction_MathQuaternionFromAxisAndAngle(Axis=(X=0,Y=1,Z=0), Angle=&RigVMFunction_MathDoubleMul(A=@"Slope Angles Radians_1".Y Angle, B=@n3.Value).Result) {template="FromAxisAndAngle::Execute(in Axis,in Angle,out Result)"}.Result
   ) {template="Make Absolute::Execute(in Local,in Parent,out Global)"}
   RigVMFunction_MathTransformMake = /Script/RigVM.RigVMFunction_MathTransformMake(Scale=(X=1,Y=1,Z=1), Rotation=@n4.Global.Rotation, Translation=RigVMVariableNode() {variable=Hit Location}.Value) {is=struct RigVMUnitNode}
   SpringInterp = &RigUnit_SpringInterpVectorV2(
    CriticalDamping=1
    Force=(X=0,Y=0,Z=0)
    Strength=4
    TargetVelocityAmount=0
    bInitializeFromTarget=False
    bUseCurrentInput=False
    Current=RigVMDispatchNode(Index=0, Array=RigVMVariableNode() {variable=Body Original Transform}.Value) {template="DISPATCH_RigVMDispatch_ArrayGetAtIndex(in Array,in Index,out Element)"}.Element.Translation
    Target=/Script/RigVM.RigVMFunction_MathTransformTransformVector(
     Location=(X=0,Y=62.404,Z=0)
     Location.Y=&RigVMFunction_MathTransformMakeRelative(Global=@n4.Global, Parent=@RigVMFunction_MathTransformMake.Result).Local.Translation.Y
     Location.Z=@n1.Result
     Transform=@RigVMFunction_MathTransformMake.Result
    ) {is=struct RigVMUnitNode}.Result
   ) {template="SpringInterp::Execute(in Target,in Strength,in CriticalDamping,in Force,in bUseCurrentInput,in Current,in TargetVelocityAmount,in bInitializeFromTarget,out Result,out Velocity)"}
   > &RigUnit_SetTransform(
    Item=(Type=Bone,Name="root")
    Value=(Rotation=(X=0,Y=0,Z=0,W=1),Translation=(X=0,Y=0,Z=0),Scale3D=(X=1,Y=1,Z=1))
    Weight=1
    bInitial=False
    Value.Rotation=@RigVMFunction_MathTransformMake.Result.Rotation
    Value.Translation=@SpringInterp.Result
   )
   > RigVMFunctionReturnNode()
  ]}
 New Function_ContainedGraph {nodes [
   n1 = RigVMFunctionEntryNode()
   RigVMFunction_MathVectorAngle = /Script/RigVM.RigVMFunction_MathVectorAngle(A=(X=0,Y=0,Z=0), B=(X=0,Y=0,Z=1), A.Y=@n1.Normal.Y, A.Z=@n1.Normal.Z) {is=struct RigVMUnitNode, template="AngleBetween::Execute(in A,in B,out Result)"}
   Sign = /Script/RigVM.RigVMFunction_MathDoubleSign(Value=@n1.Normal.Y) {is=struct RigVMUnitNode, template="Sign::Execute(in Value,out Result)"}
   Multiply = &RigVMFunction_MathFloatMul(A=@RigVMFunction_MathVectorAngle.Result, B=@Sign.Result) {template="Multiply::Execute(in A,in B,out Result)"}
   > RigVMFunctionReturnNode(
    X Angle=@Multiply.Result
    Y Angle=&RigVMFunction_MathFloatMul(
     A=&RigVMFunction_MathVectorAngle(A=(X=0,Y=0,Z=0), B=(X=0,Y=0,Z=1), A.X=@n1.Normal.X, A.Z=@n1.Normal.Z) {template="AngleBetween::Execute(in A,in B,out Result)"}.Result
     B=&RigVMFunction_MathDoubleSign(Value=@n1.Normal.X) {template="Sign::Execute(in Value,out Result)"}.Result
    ) {template="Multiply::Execute(in A,in B,out Result)"}.Result
   )
  ]}
 New Function_ContainedGraph {nodes [
   n1 = RigVMFunctionEntryNode()
   > /Script/RigVM.RigVMFunction_DebugLineNoSpace(
    DebugDrawSettings=(DepthPriority=SDPG_Foreground,Lifetime=-1)
    Thickness=1
    WorldOffset=(Rotation=(X=0,Y=0,Z=0,W=1),Translation=(X=0,Y=0,Z=0),Scale3D=(X=1,Y=1,Z=1))
    bEnabled=true
    Color=@n1.Color
    B=@n1.Hit Location
    A=@n1.Start Position
   ) {is=struct RigVMUnitNode}
   Scale = /Script/RigVM.RigVMFunction_MathVectorScale(Factor=@n1.Factor, Value=@n1.Hit Normal) {is=struct RigVMUnitNode, template="Scale::Execute(in Value,in Factor,out Result)"}
   Add = &RigVMFunction_MathVectorAdd(A=@n1.Hit Location, B=@Scale.Result) {template="Add::Execute(in A,in B,out Result)"}
   > RigVMFunction_DebugLineNoSpace = &RigVMFunction_DebugLineNoSpace(
    DebugDrawSettings=(DepthPriority=SDPG_Foreground,Lifetime=-1)
    Thickness=2
    WorldOffset=(Rotation=(X=0,Y=0,Z=0,W=1),Translation=(X=0,Y=0,Z=0),Scale3D=(X=1,Y=1,Z=1))
    bEnabled=true
    B=@Add.Result
    Color=@n1.Color
    A=@n1.Hit Location
   )
   n2 = /Script/ControlRig.RigUnit_AimBoneMath(
    DebugSettings=(bEnabled=False,Scale=10,WorldOffset=(Rotation=(X=0,Y=0,Z=0,W=1),Translation=(X=0,Y=0,Z=0),Scale3D=(X=1,Y=1,Z=1)))
    InputTransform=(Rotation=(X=0,Y=0,Z=0,W=1),Translation=(X=0,Y=0,Z=0),Scale3D=(X=1,Y=1,Z=1))
    Primary=(Weight=1,Axis=(X=0,Y=0,Z=1),Target=(X=1,Y=0,Z=0),Kind=Direction,Space=(Type=Bone,Name="None"))
    Secondary=(Weight=1,Axis=(X=1,Y=0,Z=0),Target=(X=1,Y=0,Z=0),Kind=Direction,Space=(Type=Bone,Name="None"))
    Weight=1
    InputTransform.Translation=@n1.Hit Location
    Primary.Target=@n1.Hit Normal
   ) {is=struct RigVMUnitNode}
   RigVMFunction_MathQuaternionFromEuler = &RigVMFunction_MathQuaternionFromEuler(Euler=(X=0,Y=0,Z=0))
   > RigVMFunction_DebugArcNoSpace = /Script/RigVM.RigVMFunction_DebugArcNoSpace(
    DebugDrawSettings=(DepthPriority=SDPG_Foreground,Lifetime=-1)
    Detail=16
    MaximumDegrees=360
    MinimumDegrees=0
    Thickness=1
    WorldOffset=(Rotation=(X=0,Y=0,Z=0,W=1),Translation=(X=0,Y=0,Z=0),Scale3D=(X=1,Y=1,Z=1))
    bEnabled=True
    Color=@n1.Color
    Radius=@n1.Radius
    Transform=&RigVMFunction_MathTransformMakeAbsolute(
     Local=(Rotation=(X=0,Y=0,Z=0,W=1),Translation=(X=0,Y=0,Z=0),Scale3D=(X=1,Y=1,Z=1))
     Parent=@n2.Result
     Local.Rotation=@RigVMFunction_MathQuaternionFromEuler.Result
    ) {template="Make Absolute::Execute(in Local,in Parent,out Global)"}.Global
   ) {is=struct RigVMUnitNode}
   > RigVMFunctionReturnNode()
  ]}
 RigVMFunctionLibrary {nodes [
   Draw Trace = RigVMCollapseNode(
    Color=(R=1,G=0,B=0,A=1)
    ExecuteContext=()
    Factor=20
    Hit Location=(X=0,Y=0,Z=0)
    Hit Normal=(X=0,Y=0,Z=0)
    Radius=25
    Start Position=(X=0,Y=0,Z=0)
   )
   Root Alignment = RigVMCollapseNode(
    Crouching Amount=-35
    ExecuteContext=()
    Hit Location=(X=0,Y=0,Z=0)
    Hit Normal=(X=0,Y=0,Z=0)
    Item=(Type=Bone,Name="root")
    Side Rotation Factor=0.25
   )
   Slope Angles Radians = RigVMCollapseNode(Normal=(X=0,Y=0,Z=0))
   Trace From Item = RigVMCollapseNode(
    Bone=(Type=Bone,Name="spine_01")
    Debug Control=(Type=Control,Name="Slope")
    ExecuteContext=()
    Hit Location=(X=0,Y=0,Z=0)
    Hit Normal=(X=0,Y=0,Z=0)
   )
   Trace From Location = RigVMCollapseNode(
    Control=(Type=Bone,Name="None")
    ExecuteContext=()
    Max Distance=100
    Previous Location=(X=0,Y=0,Z=0)
    Previous Normal=(X=0,Y=0,Z=0)
    Start Location=(X=0,Y=0,Z=0)
    Use Control=False
   )
  ]}
 RigVMFunction_Sequence_1_ContainedGraph {nodes [
   n1 = RigVMFunctionEntryNode()
   RigVMFunction_Sequence = &RigVMFunction_Sequence(ExecuteContext=@n1.ExecuteContext)
   n2 = &RigVMFunction_Sequence(ExecuteContext=@RigVMFunction_Sequence.B)
   RigVMFunctionReturnNode(A=@RigVMFunction_Sequence.A, B=@n2.A, C=@n2.B)
  ]}
 RigVMModel {nodes [
   RigUnit_BeginExecution = &RigUnit_BeginExecution()
   Trace From Item = RigVMFunctionReferenceNode(
    Bone=(Type=Bone,Name="root")
    Debug Control=(Type=Control,Name="Slope")
    ExecuteContext=@RigUnit_BeginExecution.ExecutePin
    Hit Location=RigVMVariableNode() {variable=Hit Location}.Value
    Hit Normal=RigVMVariableNode() {variable=Hit Normal}.Value
   )
   > RigVMVariableNode(Value=@"Trace From Item".Hit Location Out) {variable=Hit Location}
   > RigVMVariableNode(Value=@"Trace From Item".Hit Norma Out) {variable=Hit Normal}
   Make_Array = RigVMDispatchNode(Values=((Type=Bone,Name="root"))) {template="DISPATCH_RigVMDispatch_ArrayMake(in Values,out Array)"}
   > RigVMVariableNode(Value=/Script/ControlRig.RigUnit_GetTransformItemArray(Space=GlobalSpace, bInitial=False, Items=@Make_Array.Array) {is=struct RigVMUnitNode}.Transforms) {variable=Body Original Transform}
   > Root Alignment = RigVMFunctionReferenceNode(
    Crouching Amount=-35
    Item=(Type=Bone,Name="root")
    Side Rotation Factor=1
    Hit Location=RigVMVariableNode() {variable=Hit Location}.Value
    Hit Normal=RigVMVariableNode() {variable=Hit Normal}.Value
   )
  ]}
 RigVMModel Construction Graph {nodes [
   n1 = &RigUnit_PrepareForExecution()
   n2 = &RigUnit_HierarchyImportFromSkeleton(NameSpace=None, bIncludeCurves=False, bIncludeMeshSockets=False, bIncludeVirtualBones=True, ExecutePin=@n1.ExecutePin)
   RigVMFunction_ControlFlowBranch = &RigVMFunction_ControlFlowBranch(ExecuteContext=@n2.ExecutePin, Condition=RigVMVariableNode() {variable=Debug}.Value)
   /Script/ControlRig.RigUnit_HierarchyAddControlTransform(
    InitialValue=(Rotation=(X=0,Y=0,Z=0,W=1),Translation=(X=0,Y=0,Z=0),Scale3D=(X=1,Y=1,Z=1))
    Name=Slope
    OffsetSpace=LocalSpace
    OffsetTransform=(Rotation=(X=0,Y=0,Z=0,W=1),Translation=(X=0,Y=0,Z=0),Scale3D=(X=1,Y=1,Z=1))
    Parent=(Type=Bone,Name="None")
    Settings="(Shape=(bVisible=True,Name=\"Square_Thick\",Color=(R=1,G=0,B=0,A=1),Transform=(Rotation=(X=0,Y=0,Z=0,W=1),Translation=(X=0,Y=0,Z=0),Scale3D=(X=50,Y=50,Z=1))),Proxy=(bIsProxy=False,ShapeVisibility=BasedOnSelection),FilteredChannels=(),DisplayName=\"None\",bSele ...(truncated)"
    ExecutePin=@RigVMFunction_ControlFlowBranch.True
   ) {
    is=struct RigVMUnitNode
    template="SpawnControl::Execute(in InitialValue,in Settings,in OffsetTransform,in OffsetSpace,in Parent,in Name,out Item)"
   }
  ]}
]
variables [
 Body Original Transform {type="TArray<FTransform>"}
 Debug {type=bool, default=False}
 GroundTraceHit {type=bool, default=False}
 Hit Location {type=FVector, default=(X=0,Y=0,Z=0)}
 Hit Normal {type=FVector, default=(X=0,Y=0,Z=0)}
 TraceHeight {type=double, default=100}
 WorldLocation {type=FVector, default=(X=0,Y=0,Z=0)}
]
