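# Animation Blueprint for the medium TidalCrane swarm NPC, using the TidalCrane Brute skeleton.
# A GroundMovement state machine feeds DefaultSlot, then the TidalCrane_Medium Control Rig, before reaching the output pose.
# Idle loops to TurnInPlace on DoTurnInPlace and back by automatic sequence rule, though TurnInPlace has no animation.
# The event graph is unfinished: its update event is unwired, leaving ground-check and debug-print nodes disconnected from execution.
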AnimBlueprint /ProtoSwarm/NPCs/Swarm/TidalCrane_Swarm_Medium/Pawns/TidalCrane_Medium_AnimBP
target_skeleton=/ProtoSwarm/Assets/TidalCrane_Brute/Meshes/TidalCrane_Brute_Creature_Skeleton
graphs [
 AnimGraph {nodes [
   AnimGraphNode_StateMachine_0 = GroundMovement() {is=title AnimGraphNode_StateMachine}
   n1 = Control Rig(Source=Slot 'DefaultSlot'(Source=@AnimGraphNode_StateMachine_0)) {
    alphaBoolBlend {blendOption=Linear}
    controlRigAssetReference {blueprintRigClass=/ProtoSwarm/Assets/TidalCrane_Brute/Animation/ControlRig/TidalCrane_Medium_CtrlRig.TidalCrane_Medium_CtrlRig_C}
   }
   Output Pose(Result=@n1)
  ]}
 GroundMovement {
  entry=Idle
  path=AnimGraph.AnimGraphNode_StateMachine_0.GroundMovement
  ids=id
  nodes [
   AnimStateNode_0 = Idle()
   AnimStateNode_1 = TurnInPlace() {is=title AnimStateNode}
   AnimStateTransitionNode_0 = AnimStateTransitionNode()
   AnimStateTransitionNode_1 = AnimStateTransitionNode() {bAutomaticRuleBasedOnSequencePlayerInState=true}
  ]
  links [
   {
    dst=TurnInPlace
    rule=AnimGraph.AnimGraphNode_StateMachine_0.GroundMovement.AnimStateTransitionNode_0.Transition
    src=Idle
    via=AnimStateTransitionNode_0
   }
   {
    dst=Idle
    rule=AnimGraph.AnimGraphNode_StateMachine_0.GroundMovement.AnimStateTransitionNode_1.Transition
    src=TurnInPlace
    via=AnimStateTransitionNode_1
   }
  ]
 }
 Idle {
  path=AnimGraph.AnimGraphNode_StateMachine_0.GroundMovement.AnimStateNode_0.Idle
  nodes [Output Animation Pose(Result=Sequence Player 'TidalCrane_Brute_Idle'() {sequence=/ProtoSwarm/Assets/TidalCrane_Brute/Animation/Locomotion/TidalCrane_Brute_Idle})]
 }
 TurnInPlace {path=AnimGraph.AnimGraphNode_StateMachine_0.GroundMovement.AnimStateNode_1.TurnInPlace, nodes [Output Animation Pose()]}
 Transition {
  path=AnimGraph.AnimGraphNode_StateMachine_0.GroundMovement.AnimStateTransitionNode_0.Transition
  nodes [Result(bCanEnterTransition=Get DoTurnInPlace() -> (DoTurnInPlace:bool))]
 }
 Transition {path=AnimGraph.AnimGraphNode_StateMachine_0.GroundMovement.AnimStateTransitionNode_1.Transition, nodes [Result()]}
 EventGraph {nodes [
   TryGetPawnOwner()
   n1 = Get CharacterMovement!(self:Character Object Reference=TryGetFortAIPawn!() -> (ReturnValue:Fort AIPawn Object Reference)) -> (CharacterMovement:Character Movement Component Object Reference)
   GetBoneTransform!(
    self:Skinned Mesh Component Object Reference=Get Mesh(self=GetCharacterOwner!(self:Character Movement Component Object Reference=@n1) -> (ReturnValue:Character Object Reference))
    InBoneName:Name=Root
    TransformSpace:ERelativeTransformSpace Enum=RTS_World
   ) -> (ReturnValue:Transform)
   n2 = IsMovingOnGround!(self:Nav Movement Component Object Reference=@n1) -> (ReturnValue:bool)
   PrintString!(
    execute
    InString="To String (Boolean)"(InBool=Set IsOnGround(IsOnGround:bool=@n2) -> (Output_Get:bool))
    bPrintToScreen
    bPrintToLog
    TextColor
    Duration
    Key
   ) -> (then)
   "To String (Float)"(InDouble=Get Speed() -> (Speed:float))
   Event BlueprintUpdateAnimation()
  ]}
]
variables [
 Aim Yaw {type=number}
 DoTurnInPlace {type=boolean}
 IsOnGround {type=boolean}
]
